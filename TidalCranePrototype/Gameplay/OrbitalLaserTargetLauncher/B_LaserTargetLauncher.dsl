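; Orbital laser target launcher Blueprint that collects sticky target beacons and calls in a laser strike on them.
; Each beacon impact adds the projectile to a target array and sets a short timer on the same event, so successive impacts restart the countdown.
; When it expires, Disconnected spawns a LaserActor at the launcher's location, passes it the target array and a test location array, then clears the targets.
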(event Custom|TargetBeaconImpacted
  (params (TargetRef :type B%20Prj%20Sticky%20Laser%20Target%20Object%20Reference))
  (Utilities|Array|Add (Variables|Default|GetLaserTargetArray) TargetRef)
  (Utilities|Time|SetTimerbyEvent (delegate Custom|Disconnected) 1.5))

(event Custom|Disconnected
  (Game|SpawnActorfromClass "/TidalCranePrototype/Gameplay/OrbitalLaserTargetLauncher/Blueprints/LaserActor.LaserActor_C" (Math|Transform|MakeTransform (Transformation|GetActorLocation)) "Undefined" "MultiplyWithRoot" 0 (Variables|Default|GetLaserTargetArray) (Variables|Default|Gettest_LocationArray) 0)
  (Utilities|Array|Clear (Variables|Default|GetLaserTargetArray)))

(fn ConstructionScript
  (|CallParentFunction))
