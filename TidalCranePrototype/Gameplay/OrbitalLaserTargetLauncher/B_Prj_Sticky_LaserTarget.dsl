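; Sticky target-beacon projectile for the orbital laser target launcher, with team-colored fuse FX in flight.
; On touch, the server attaches it to allowed surfaces or pawns and calls TargetBeaconImpacted on the launcher once, or bounces it off disallowed pawns.
; Stopping halts rotation and plays stick sounds with indicators; detaching resumes simulation, and the OnRep re-allows sticking.
; Water contact stops it with splash FX; explosion and stuck-FX events are empty stubs.
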(event EventBeginPlay
  ;[1 1 1] BeginPlay - Play some FX, set up timers.
  (Utilities|FlowControl|Sequence
    (:then_0
      (Networking|CanExecuteCosmeticEvents
        (:True
          (CallFunction|PlayinAirFuseFX))))
    (:then_1
      (Utilities|FlowControl|Delay (Variables|Default|GetSelfStickDelay))
      (Variables|Default|SetIsSelfStickDelayOver? true))
    (:then_2)))

(event EventEndPlay
  (params (EndPlayReason :type EEndPlayReason%20Enum :default "Destroyed"))
  ;[1 1 1] EndPlay - Remove AV
  (|CallParentFunction EndPlayReason)
  (SoundIndicator|StopSoundIndicatorLooping (Variables|Default|GetFortSoundIndicator)))

(event Custom|PlayInAirFuseFX
  (bind _beacon_core_color__enemy (Variables|FX|GetBeacon%20Core%20Color%20%28Enemy%29))
  (bind _beacon_core_color__friendly (Variables|FX|GetBeacon%20Core%20Color%20%28Friendly%29))
  (bind _color_out (Utilities|GetTeamVFX _beacon_core_color__friendly _beacon_core_color__enemy))
  (bind _returnvalue (Math|Conversions|ToVector(LinearColor) _color_out))
  ;[1 1 1] Set up and play initial in-air fuse fx
  (Effects|Components|ParticleSystem|SetVectorParameter (Variables|B_Prj_Grenade|GetFuse_Particle) "CoreLightColor" _returnvalue) ; We no longer use team color, \nshould appear red for all players
  (Effects|Components|ParticleSystem|SetVectorParameter (Variables|B_Prj_Grenade|GetFuse_Particle) "StreakColor" _returnvalue)
  (Effects|Components|ParticleSystem|SetColorParameter (Variables|B_Prj_Grenade|GetFuse_Particle) "TeamColor" (Utilities|GetTeamVFX (Variables|FX|GetBeacon%20Streak%20Color%20%28Friendly%29) (Variables|FX|GetBeacon%20Streak%20Color%20%28Enemy%29))) ; We no longer use team color, \nshould appear red for all players
  ;;[0.6458 0.6458 0.6458] Set Material based on Team
  (bind _returnvalue_2 (Utilities|Time|SetTimerbyEvent (delegate Custom|TickFuseParticle) 0.15 true))
  (Variables|Default|SetFuseParticleTimer _returnvalue_2)
  (Components|Activation|Activate (Variables|B_Prj_Grenade|GetFuse_Particle) false))

(event Custom|TickFuseParticle
  ;[1 1 1] Set up and play initial in-air fuse fx
  ;;[0.6458 0.6458 0.6458] Set Material based on Team
  (Components|Activation|Activate (Variables|B_Prj_Grenade|GetFuse_Particle) false))

(event EventOnTouched
  (params (OtherActor :type Actor%20Object%20Reference) (OtherComp :type Primitive%20Component%20Object%20Reference) (HitResult :type Hit%20Result%20Structure%20%28by%20ref%29) (bIsOverlap :type Boolean))
  (bind (_returnvalue_x _returnvalue_y _returnvalue_z) (Math|Vector|BreakVector (Transformation|GetActorLocation)))
  (bind (_returnvalue_x_2 _returnvalue_y_2) (Math|Vector|BreakVector (Transformation|GetActorLocation (Variables|Default|GetTouchedActor))))
  (bind (_bblockinghit _binitialoverlap _time _distance _location _impactpoint _normal _impactnormal _physmat _hitactor _hitcomponent) (Collision|BreakHitResult HitResult))
  (bind (_aspawn _bsuccess) (Utilities|Casting|CastToPawn (Variables|Default|GetTouchedActor)))
  (bind (_asfort_water_body_actor _bsuccess_2) (Utilities|Casting|CastToFortWaterBodyActor OtherActor))
  (bind _projectilemovementcomponent (Variables|Movement|GetProjectileMovementComponent))
  ;[1 1 1] EventOnTouched
  (Variables|Default|SetTouchedActor OtherActor)
  (if _bsuccess_2
    (CallFunction|HandleEnterWater HitResult)
    (else
      (Networking|SwitchHasAuthority
        (:Authority
          (Utilities|SticktoAllyCheck OtherActor
            (:StickToActor
              (if (not (Components|ComponentHasTag _hitcomponent "NoStick"))
                (Variables|Projectile|SetbStopSimulatingOnHit true)
                (Utilities|FlowControl|DoOnce false
                  (:Completed
                    (if (or (Variables|Default|GetStickToPlayers) (not _bsuccess))
                      (Gameplay|AuthorityAttachtoActor (Variables|Default|GetFortAttachableProjectile) HitResult (Variables|Default|GetTouchedActor) (ScalableFloat|GetValueatLevel (Variables|Default|GetOffsetDistanceFromPhysicsMesh)) (Variables|Default|GetOffsetDistanceFromBone) 0.0)
                      (Utilities|FlowControl|DoOnce false
                        (:Completed
                          (bind _returnvalue (CallFunction|GetOwnerWeapon))
                          (bind _asb_laser_target_launcher (Utilities|Casting|CastToB_LaserTargetLauncher _returnvalue))
                          (Class|BLaserTargetLauncher|TargetBeaconImpacted _asb_laser_target_launcher self)))
                      (else
                        (Variables|/Script/Engine.ProjectileMovementComponent|SetProjectileGravityScale 0.3 _projectilemovementcomponent)
                        (Projectile|SetVelocity _projectilemovementcomponent (Utilities|CalculateBounceOffVelocity))
                        (Transformation|SetActorLocation 0 (Math|Vector|MakeVector _returnvalue_x_2 _returnvalue_y_2 _returnvalue_z) false false)))))))
            (:NoAllyStick
              (Variables|Projectile|SetbStopSimulatingOnHit false))))))))

(event Custom|ReallowSticking
  (bind (_returnvalue_x _returnvalue_y _returnvalue_z) (Math|Vector|BreakVector (Transformation|GetActorLocation)))
  (bind (_returnvalue_x_2 _returnvalue_y_2) (Math|Vector|BreakVector (Transformation|GetActorLocation (Variables|Default|GetTouchedActor))))
  (bind (_aspawn _bsuccess) (Utilities|Casting|CastToPawn (Variables|Default|GetTouchedActor)))
  (bind _projectilemovementcomponent (Variables|Movement|GetProjectileMovementComponent))
  ;[1 1 1] EventOnTouched
  (Utilities|FlowControl|DoOnce false
    (:Completed
      (if (or (Variables|Default|GetStickToPlayers) (not _bsuccess))
        (Gameplay|AuthorityAttachtoActor (Variables|Default|GetFortAttachableProjectile) (stale EventOnTouched HitResult) (Variables|Default|GetTouchedActor) (ScalableFloat|GetValueatLevel (Variables|Default|GetOffsetDistanceFromPhysicsMesh)) (Variables|Default|GetOffsetDistanceFromBone) 0.0)
        (Utilities|FlowControl|DoOnce false
          (:Completed
            (bind _returnvalue (CallFunction|GetOwnerWeapon))
            (bind _asb_laser_target_launcher (Utilities|Casting|CastToB_LaserTargetLauncher _returnvalue))
            (Class|BLaserTargetLauncher|TargetBeaconImpacted _asb_laser_target_launcher self)))
        (else
          (Variables|/Script/Engine.ProjectileMovementComponent|SetProjectileGravityScale 0.3 _projectilemovementcomponent)
          (Projectile|SetVelocity _projectilemovementcomponent (Utilities|CalculateBounceOffVelocity))
          (Transformation|SetActorLocation 0 (Math|Vector|MakeVector _returnvalue_x_2 _returnvalue_y_2 _returnvalue_z) false false))))))

(event OnProjectileStop(ProjectileMovementComponent)
  (params (ImpactResult :type Hit%20Result%20Structure%20%28by%20ref%29))
  (bind (_bblockinghit _binitialoverlap _time _distance _location _impactpoint _normal _impactnormal _physmat _hitactor) (Collision|BreakHitResult ImpactResult))
  (bind (_asfort_player_pawn _bsuccess) (Utilities|Casting|CastToFortPlayerPawn _hitactor))
  ;[1 1 1] OnStop - Stop the rotation and bind to destruction of building actor if relevant.
  (Variables|Default|SetbSimulationRunning false)
  (Variables|Default|SetHitNormal _impactnormal)
  (Variables|Default|SetHitLocation (select _bblockinghit _location (Transformation|GetActorLocation)))
  (CallFunction|StopRotation)
  (Networking|CanExecuteCosmeticEvents
    (:True
      (CallFunction|PlayStickSound (Variables|Default|GetHitLocation) _hitactor _bsuccess)
      (CallFunction|PlayStuckFuseFX))))

(event Custom|Stop_Rotation
  ;[1 1 1] Stop Rotating Movement
  (Components|Movement|SetUpdatedComponent (Variables|Default|GetRotatingMovement) 0)
  (Collision|SetCollisionEnabled (Variables|Collision|GetCapsuleComponent) "NoCollision")
  (Networking|SwitchHasAuthority
    (:Authority
      (Transformation|SetActorLocationAndRotation (Variables|Default|GetHitLocation) (Math|Conversions|RotationFromXVector (Math|Vector|NegateVector (Variables|Default|GetHitNormal)))))))

(event Custom|SetExplosionTimer)

(event Custom|PlayStuckFuseFX)

(event OnProjectileDetached(FortAttachableProjectile)
  (params (HostActor :type Actor%20Object%20Reference) (DetachReason :type EFortProjectileDetachReason%20Enum :default "HostDied"))
  ;[1 1 1] Attached target destroyed. Start simulating again.
  (Utilities|Set|Add (Variables|Default|GetIgnore) HostActor)
  (Variables|Default|SetbSimulationRunning true)
  (Networking|ForceNetUpdate)
  (Projectile|ResumeSimulation))

(event EventEvent_Building_Actor_Destroyed
  ;[1 1 1] Building Actor Destroyed Interface
  (Gameplay|EjectfromAttachedtoActor (Variables|Default|GetFortAttachableProjectile)))

(event EventOnExploded
  (params (HitActors :type Array%20of%20Actor%20Object%20References) (HitResults :type Array%20of%20Hit%20Result%20Structures)))

(event Custom|Event_TriggerExplosion
  ;[1 0.6838 0] Event Trigger Explosion
  (Utilities|FlowControl|DoOnce false))

(event EventDestroyed
  ;[1 0.6838 0] Event Trigger Explosion
  (Utilities|FlowControl|DoOnce false))

(event Custom|HandleEnterWater
  (params (WaterHit :type Hit%20Result%20Structure))
  ;[1 1 1] Handle Entering Water's event like trigger the beeping and the explosion event
  (Utilities|FlowControl|DoOnce false
    (:Completed
      (Variables|/Script/Engine.ProjectileMovementComponent|SetbRotationFollowsVelocity false (Variables|Movement|GetProjectileMovementComponent))
      (CallFunction|SetExplosionTimer)
      (Game|Components|ProjectileMovement|StopSimulating (Variables|Movement|GetProjectileMovementComponent) WaterHit))))

(event OnEnterWaterEvent(WaterInteractionComponent)
  (params (WaterBody :type Fort%20Water%20Body%20Actor%20Object%20Reference) (WaterInteractionComponent :type Fort%20Water%20Interaction%20Component%20Object%20Reference) (bIsFirstBody :type Boolean))
  ;[1 1 1] Handle Entering Water, Play splash FX and Sounds
  (Utilities|FlowControl|DoOnce false
    (:Completed
      (CallFunction|PlaySplashFX WaterInteractionComponent)
      (CallFunction|PlayStuckFuseFX))))

(fn AttemptAutoExplode
  (params (HitObject :type Object%20Reference) (HitComponent :type Primitive%20Component%20Object%20Reference) (bOverlap :type Boolean)))

(fn OnRep_bResumeSimulation
  (if (Variables|Default|GetbSimulationRunning)
    (Variables|/Script/Engine.PrimitiveComponent|SetbGenerateOverlapEvents true (Variables|Collision|GetCapsuleComponent))
    (Components|Movement|SetUpdatedComponent (Variables|Movement|GetProjectileMovementComponent) (Variables|Collision|GetCapsuleComponent))
    (Collision|SetCollisionEnabled (Variables|Collision|GetCapsuleComponent) "QueryOnly")
    (CallFunction|ReallowSticking)
    (Networking|CanExecuteCosmeticEvents
      (:True
        (Components|Movement|SetUpdatedComponent (Variables|Default|GetRotatingMovement) (Variables|B_Prj_Grenade|GetMesh))))))

(fn PlaySplashFX
  (params (WaterInteractionComponent :type Fort%20Water%20Interaction%20Component%20Object%20Reference))
  (bind (_waterplanelocation _waterplanenormal _watersurfaceposition _waterdepth _waterbodyidx) (FX|GetLastWaterSurfaceInfo WaterInteractionComponent :!class /Script/FortniteGame.FortWaterInteractionComponent))
  (Variables|Default|SetHitLocation (Transformation|GetActorLocation))
  (Variables|Default|SetHitNormal "0, 0,1.000000")
  (bind _returnvalue (Effects|SpawnEmitteratLocation "/Game/Weapons/FORT_Generic_Effects/Impacts/Water/P_Pistol_Impact_Water_Apollo.P_Pistol_Impact_Water_Apollo" _watersurfaceposition (Math|Rotator|MakeRotfromX _waterplanenormal)))
  (Effects|Components|ParticleSystem|SetFloatParameter _returnvalue "WaterDepth" _waterdepth)
  (Effects|Components|ParticleSystem|SetFloatParameter _returnvalue "WaterBodyIndex" (Math|Conversions|ToFloat(Integer) _waterbodyidx))
  (Audio|PlaySoundatLocation "/Game/Sounds/Fort_GamePlay_Sounds/Explosives/Grenade_Projectile_Bounce_Water_Splash_NoLoop_Cue.Grenade_Projectile_Bounce_Water_Splash_NoLoop_Cue" (Transformation|GetActorLocation)))

(fn PlayStickSound
  (params (Location :type Vector) (HitActor :type Actor%20Object%20Reference) (Success :type Boolean))
  (bind _returnvalue (Game|GetInstigator))
  (bind _audio_stucktoplayer (Variables|Default|GetAudio_StuckToPlayer))
  (Audio|Components|Audio|FadeOut (Variables|Default|GetInAirLoop) 0.2 0.0 "Linear" :!class /Script/Engine.AudioComponent)
  (if Success
    (Utilities|IsValid _returnvalue
      (:Is%20Valid
        (if (Pawn|IsLocallyControlled _returnvalue)
          (Audio|PlaySound2D (Variables|Default|GetAudio_StuckToPlayer)) ; Plays on player who threw grenade
          (else
            (Utilities|IsValid HitActor
              (:Is%20Valid
                (if (Pawn|IsLocallyControlled (Utilities|Casting|CastToFortPlayerPawn HitActor))
                  (Audio|PlaySound2D _audio_stucktoplayer) ; Plays on player hit with grenade
                  (SoundIndicator|StartTargetedSoundIndicatorLooping (Variables|Default|GetFortSoundIndicator) HitActor _audio_stucktoplayer 1.0 0)
                  (else
                    _)))
              (:Is%20Not%20Valid
                _)))))
      (:Is%20Not%20Valid
        _))
    (CallFunction|PlayStuckinEnvironmentSound Location)
    (else
      (CallFunction|PlayStuckinEnvironmentSound Location))))

(fn PlayStuckInEnvironmentSound
  (params (Location :type Vector))
  (bind _audio_stucktoenvironment (Variables|Default|GetAudio_StuckToEnvironment))
  (Audio|PlaySoundatLocation _audio_stucktoenvironment Location)
  (SoundIndicator|StartSoundIndicatorLooping (Variables|Default|GetFortSoundIndicator) Location _audio_stucktoenvironment 1.0 0))

(fn ConstructionScript
  (Projectile|SetGravityScale (ScalableFloat|GetValueatLevel (Variables|Default|GetProjectileGravityScale)))
  (Networking|CanExecuteCosmeticEvents
    (:True
      (Components|Movement|SetUpdatedComponent (Variables|Default|GetRotatingMovement) (Variables|B_Prj_Grenade|GetMesh)))))
