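; Orbital laser actor that sweeps a beam across an ordered list of target location actors, damaging what it touches.
; Each tick moves the beam endpoint at constant speed, snaps it to surfaces by line trace, and drags audio and camera shake along.
; Distance and min/max timing trigger killing each target and advancing; the laser destroys itself after lingering on the last.
; A server-only looping timer sphere-traces along the beam and applies GE_LaserDamage via the instigator's ability system.
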(event EventBeginPlay
  (CameraShake|Start (Variables|Default|GetCameraShakeSource))
  (bind _output_get (Variables|Default|SetDesiredStopLocation (Transformation|GetActorLocation (Utilities|Array|Get(acopy) (Variables|Default|GetTargetLocations)))))
  (Effects|Components|ParticleSystem|SetVectorParameter (Variables|Default|GetLaser) "Stop Location" _output_get)
  (Networking|SwitchHasAuthority
    (:Authority
      (Utilities|Time|SetTimerbyEvent (EventDispatchers|CreateEvent :!event Custom|DoLaserDamage) 0.1 true))))

(event EventTick
  (params (DeltaSeconds :type Float%20%28single-precision%29))
  (bind _nexttargetindex (Variables|Default|GetNextTargetIndex))
  (bind _targetlocations (Variables|Default|GetTargetLocations))
  (bind _output (Utilities|Array|Get(acopy) _targetlocations _nexttargetindex))
  (bind _returnvalue (Transformation|GetActorLocation _output))
  (Variables|Default|SetTimeElapsedBetweenLocations (+ (Variables|Default|GetTimeElapsedBetweenLocations) DeltaSeconds))
  (bind _output_get (Variables|Default|SetDesiredStopLocation (Math|Interpolation|VInterptoConstant (Variables|Default|GetDesiredStopLocation) _returnvalue DeltaSeconds (Variables|Default|GetInterpSpeed))))
  (bind _adjustedstoplocation (CallFunction|SticktoSurface _output_get))
  (bind _output_get_2 (Variables|Default|SetActualStopLocation _adjustedstoplocation))
  (Effects|Components|ParticleSystem|SetVectorParameter (Variables|Default|GetLaser) "Stop Location" (+ _output_get_2 "0.000000,0.000000,50.000000"))
  (CallFunction|MoveAudioandCameraShake)
  (if (or (and (<= (Math|Vector|DistanceSquared(Vector) (Variables|Default|GetDesiredStopLocation) _returnvalue) 1e+01) (>= (Variables|Default|GetTimeElapsedBetweenLocations) (Variables|Default|GetMinTimeBetweenLocations))) (>= (Variables|Default|GetTimeElapsedBetweenLocations) (Variables|Default|GetMaxTimeBetweenLocations)))
    (Delivery|Kill (Utilities|Array|Get(acopy) (Variables|Default|GetTargetLocations) (Variables|Default|GetNextTargetIndex)))
    (Variables|Default|SetTimeElapsedBetweenLocations 0.0)
    (bind _didupdate (CallFunction|CheckandSetNextIndex))
    (if (not _didupdate)
      (Utilities|FlowControl|DoOnce false
        (:Completed
          (Utilities|FlowControl|Delay (Variables|Default|GetFinalPointLingerTime))
          (Actor|DestroyActor))))))

(event Custom|DoLaserDamage
  (bind (_outhits _returnvalue) (Collision|MultiSphereTraceForObjects (Transformation|GetWorldLocation (Variables|Default|GetLaser)) (Variables|Default|GetActualStopLocation) 1.5e+02 (Utilities|Array|MakeArray "ObjectTypeQuery3" "ObjectTypeQuery1" "ObjectTypeQuery5" "ObjectTypeQuery1")))
  (if _returnvalue
    (Ability|Tasks|ApplyGameplayEffectToHitResultsAsync (Ability|GetFortAbilitySystemComponent (Game|GetInstigator)) _outhits "/TidalCranePrototype/Gameplay/OrbitalLaserTargetLauncher/Blueprints/GE_LaserDamage.GE_LaserDamage_C" 0 1)))

(fn CheckAndSetNextIndex
  (bind _nexttargetindex (Variables|Default|GetNextTargetIndex))
  (bind _returnvalue (+ _nexttargetindex 1))
  (if (Utilities|Array|IsValidIndex (Variables|Default|GetTargetLocations) _returnvalue)
    (Variables|Default|SetNextTargetIndex _returnvalue)
    (return true)
    (else
      (return false))))

(fn MoveAudioandCameraShake
  (bind _actualstoplocation (Variables|Default|GetActualStopLocation))
  (Transformation|SetWorldLocation (Variables|Default|GetCameraShakeSource) _actualstoplocation false false)
  (Transformation|SetWorldLocation (Variables|Default|GetAudio) _actualstoplocation false false)
  (Transformation|SetWorldLocation (Variables|Default|GetAudio2) _actualstoplocation false false))

(fn StickToSurface
  (params (InStopLocation :type Vector))
  (bind (_outhit _returnvalue) (Collision|LineTraceForObjects (Transformation|GetWorldLocation (Variables|Default|GetLaser)) (+ (Transformation|GetWorldLocation (Variables|Default|GetLaser)) (* (Math|Vector|GetUnitDirection(Vector) (Transformation|GetWorldLocation (Variables|Default|GetLaser)) (Variables|GetInStopLocation)) (* (Math|Vector|Distance(Vector) (Transformation|GetWorldLocation (Variables|Default|GetLaser)) (Variables|GetInStopLocation)) 1.5))) (Utilities|Array|MakeArray "ObjectTypeQuery1" "ObjectTypeQuery3" "ObjectTypeQuery5")))
  (if _returnvalue
    (return (.ImpactPoint (Collision|BreakHitResult _outhit)))
    (else
      (return (Variables|GetInStopLocation)))))

(fn ConstructionScript)
